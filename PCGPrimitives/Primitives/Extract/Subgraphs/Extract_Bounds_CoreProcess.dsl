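; PCG subgraph that is the core step of the Extract Bounds primitive, reducing point data to points whose bounds enclose their source points.
; It keeps only point data tagged Input:0, then branches on CombineAllInputsIntoSinglePoint.
; When true (default), Merge Points pools all inputs without merging metadata before Combine Points yields one point.
; When false, Combine Points runs per input; neither path uses the first point's transform.
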(graph Extract_Bounds_CoreProcess
  (params (CombineAllInputsIntoSinglePoint :type Bool :default true))
  ;[0 0 0 0.5 12] Primary/Secondary Inputs
  (bind FilterDataByTag_0 (Filter%20Data%20By%20Tag :SelectedTags "Input:0"))
  ;[0 0 0 0.5] Core Process
  (bind FilterDataByType_0 (Filter%20Data%20By%20Type :TargetType (struct :Ids [(struct :Struct /Script/PCG.PCGDataTypeInfoPoint)])))
  (bind Branch_0 (Branch :bOutputToB-param CombineAllInputsIntoSinglePoint))
  (bind CombinePoints_0 (Combine%20Points :bUseFirstPointTransform false))
  (bind MergePoints_0 (Merge%20Points :bMergeMetadata false))
  (bind CombinePoints_1 (Combine%20Points :bUseFirstPointTransform false))
  (connect FilterDataByType_0 InsideFilter Branch_0 In)
  (connect Branch_0 Output%20A CombinePoints_0 In)
  (connect MergePoints_0 Out CombinePoints_1 In)
  (connect DefaultInputNode In FilterDataByTag_0 In)
  (connect FilterDataByTag_0 InsideFilter FilterDataByType_0 In)
  (connect Branch_0 Output%20B MergePoints_0 In)
  (output Out CombinePoints_0)
  (output Out CombinePoints_1))
