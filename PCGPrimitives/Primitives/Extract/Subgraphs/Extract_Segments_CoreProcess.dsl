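; PCG subgraph that cuts input splines and Polygon2D shapes into spline segments, optionally at corners sharper than CutAngleThreshold.
; Polygons become closed linear splines; splines get control-point yaw reset via a temporary Polygon2D conversion.
; A threshold check selects either an index-based predicate splitting every segment, or next-point angle divided by threshold and cast to integer as split predicate.
; Temporary attributes are deleted before output.
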(graph Extract_Segments_CoreProcess
  (params (CutAngleThreshold :type Double :default 3e+01))
  (bind FilterDataByType_2 (Filter%20Data%20By%20Type :TargetType (struct :Ids [(struct :Struct /Script/PCG.PCGDataTypeInfoSpline)])))
  (bind AttributeCompareOp_1 (Attribute%20Compare%20Op :InA CutAngleThreshold))
  (bind FilterDataByType_1 (Filter%20Data%20By%20Type :TargetType (struct :Ids [(struct :Struct /Script/PCG.PCGDataTypeInfoPolygon2D)])))
  (bind ToPoint_2 (To%20Point))
  (bind CreateSpline_2 (Create%20Spline :bClosedLoop true :bLinear true :InterpTypeAttribute None))
  (bind Branch_1 (Branch))
  (bind SplitSpline_0 (Split%20Spline :Mode ByPredicateOnControlPoints)) ; Split using angle threshold predicate
  (bind AttributeMathsOp_0 (Attribute%20Maths%20Op :InputSource1 $Index :OutputTarget predicate))
  (bind SplitSpline_1 (Split%20Spline :Attribute predicate :Mode ByPredicateOnControlPoints)) ; Split all segments
  (bind DeleteAttributes_0 (Delete%20Attributes :SelectedAttributes "predicate, Ryaw, nextPointIndex, AngleToNextPoint"))
  ;[0 0 0 0.5] Force reset orientation of control points to point toward the next point
  (bind GetSplineControlPoints_0 (Get%20Spline%20Control%20Points))
  (bind CreatePolygon2D_0 (Create%20Polygon%202D))
  (bind CopyAttributes_0 (Copy%20Attributes :InputSource $LocalRotation.Yaw :OutputTarget $LocalRotation.Yaw))
  ;[0 0 0 0.5] Compute angle to next point and use threshold to create split spine predicate
  (bind Subgraph_0 (Subgraph :Graph /PCGPrimitives/Primitives/Shared/Splines_CalculateAngleToNextPoint.Splines_CalculateAngleToNextPoint))
  (bind AttributeMathsOp_1 (Attribute%20Maths%20Op :Operation Divide :InB CutAngleThreshold))
  (bind AttributeCast_0 (Attribute%20Cast :OutputType Integer32))
  (connect AttributeMathsOp_1 Out AttributeCast_0 In)
  (connect Branch_1 Output%20B AttributeMathsOp_0 InA)
  (connect Subgraph_0 Out AttributeMathsOp_1 InA)
  (connect AttributeCompareOp_1 Out Branch_1 bOutputToB)
  (connect CopyAttributes_0 Out Branch_1 In)
  (connect CreateSpline_2 Out Branch_1 In)
  (connect CreatePolygon2D_0 Out CopyAttributes_0 Source)
  (connect FilterDataByType_2 InsideFilter CopyAttributes_0 Target)
  (connect GetSplineControlPoints_0 Out CreatePolygon2D_0 In)
  (connect ToPoint_2 Out CreateSpline_2 In)
  (connect SplitSpline_0 Out DeleteAttributes_0 In)
  (connect SplitSpline_1 Out DeleteAttributes_0 In)
  (connect DefaultInputNode In FilterDataByType_1 In)
  (connect DefaultInputNode In FilterDataByType_2 In)
  (connect FilterDataByType_2 InsideFilter GetSplineControlPoints_0 In)
  (connect AttributeCast_0 Out SplitSpline_0 In)
  (connect AttributeMathsOp_0 Out SplitSpline_1 In)
  (connect Branch_1 Output%20A Subgraph_0 In)
  (connect FilterDataByType_1 InsideFilter ToPoint_2 In)
  (output Out DeleteAttributes_0))
